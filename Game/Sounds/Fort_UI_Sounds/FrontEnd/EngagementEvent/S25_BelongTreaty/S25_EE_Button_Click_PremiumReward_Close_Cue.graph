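# Sound cue for the UI click when closing a premium reward in the Season 25 BelongTreaty engagement event front end, as its name indicates.
# A single wave player plays the shared main front-end button click wave, routed to the Fort_UI sound class.
# It overrides concurrency to cap simultaneous plays, stopping the oldest, with a short retrigger time to limit rapid repeats.
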SoundCue /Game/Sounds/Fort_UI_Sounds/FrontEnd/EngagementEvent/S25_BelongTreaty/S25_EE_Button_Click_PremiumReward_Close_Cue
root=WavePlayer_1
nodes [{id=WavePlayer_1, properties {soundWaveAssetPtr=/Game/Sounds/Fort_UI_Sounds/FrontEnd/New_FrontEnd/UI_FN_Button_Click_Main_01}}]
properties {
 bOverrideConcurrency=true
 maxDistance=2097152
 soundClassObject=&Fort_UI
 volumeMultiplier=1
 concurrencyOverrides {maxCount=3, resolutionRule=StopOldest, retriggerTime=0.05}
}
