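# Front-end UI sound cue for completing a reward purchase in the Season 25 BelongTreaty engagement event.
# A Mixer layers a loud hold-button completion sound with a rewards-modal star-appear sound, the latter slightly lowered by a fixed-volume Modulator.
# The cue plays in the Fort_UI sound class with boosted volume.
# Its concurrency override allows one instance, blocking new plays and using a short retrigger time.
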SoundCue /Game/Sounds/Fort_UI_Sounds/FrontEnd/EngagementEvent/S25_BelongTreaty/S25_EE_BuyRewards_Complete_Cue
root=Mixer_0
nodes [
 {id=Mixer_0, children [WavePlayer_1,Modulator_0], properties {inputVolume [1,1]}}
 {id=Modulator_0, children [WavePlayer_2], properties {volumeMax=0.8, volumeMin=0.8}}
 {
  id=WavePlayer_1
  properties {soundWaveAssetPtr=/Game/Sounds/Fort_UI_Sounds/FrontEnd/DefaultSounds/HoldableButtons/wav/UI_Button_Press_Hold_Loud_Complete}
 }
 {id=WavePlayer_2, properties {soundWaveAssetPtr=/Game/Sounds/Fort_UI_Sounds/FrontEnd/RewardsAndGifts/wav/UI_RewardsModal_StarAppear}}
]
properties {
 bOverrideConcurrency=true
 maxDistance=2097152
 soundClassObject=&Fort_UI
 volumeMultiplier=1.3
 concurrencyOverrides {maxCount=1, resolutionRule=PreventNew, retriggerTime=0.2}
}
