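# UI sound cue for pressing the buy-rewards button in the Season 25 BelongTreaty engagement event front end.
# A Mixer layers an event-specific press-and-hold wave over the generic loud holdable-button press wave.
# The generic wave passes through a default Modulator, then an Enveloper that fades its volume out over 2.5 seconds.
# A concurrency override allows one instance and blocks new ones, with a short retrigger time.
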SoundCue /Game/Sounds/Fort_UI_Sounds/FrontEnd/EngagementEvent/S25_BelongTreaty/S25_EE_BuyRewards_Press_Cue
root=Mixer_0
nodes [
 {
  id=Enveloper_0
  children [Modulator_0]
  properties {
   pitchCurve {editorCurveData {keys [{time=0}]}}
   volumeCurve {editorCurveData {keys [<at 0 1 RCIM_Cubic>, {arriveTangent=-0.3864, interpMode=RCIM_Cubic, leaveTangent=-0.3864, tangentMode=RCTM_User, time=2.5, value=0}]}}
  }
 }
 {id=Mixer_0, children [Enveloper_0,WavePlayer_0], properties {inputVolume [1,1]}}
 {id=Modulator_0, children [WavePlayer_1]}
 {
  id=WavePlayer_0
  properties {soundWaveAssetPtr=/Game/Sounds/Fort_UI_Sounds/FrontEnd/EngagementEvent/S25_BelongTreaty/wav/S25_EE_PressHold_Layer}
 }
 {
  id=WavePlayer_1
  properties {soundWaveAssetPtr=/Game/Sounds/Fort_UI_Sounds/FrontEnd/DefaultSounds/HoldableButtons/wav/UI_Button_Press_Hold_Loud_Pressed}
 }
]
properties {
 bOverrideConcurrency=true
 maxDistance=2097152
 soundClassObject=&Fort_UI
 volumeMultiplier=1.4
 concurrencyOverrides {maxCount=1, resolutionRule=PreventNew, retriggerTime=0.2}
}
