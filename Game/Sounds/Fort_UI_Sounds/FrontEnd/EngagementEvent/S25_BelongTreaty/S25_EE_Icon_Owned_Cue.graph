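# UI SoundCue for the owned-icon state in the Season 25 BelongTreaty front-end engagement event, routed to the Fort_UI sound class.
# Its only node is a root Modulator with slight volume variation and an empty child input, so the graph references no sound wave.
# The cue lowers overall volume and overrides concurrency to allow two instances, stopping the oldest when exceeded.
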SoundCue /Game/Sounds/Fort_UI_Sounds/FrontEnd/EngagementEvent/S25_BelongTreaty/S25_EE_Icon_Owned_Cue
root=Modulator_0
nodes [{id=Modulator_0, children [null], properties {volumeMax=1.1}}]
properties {
 bOverrideConcurrency=true
 maxDistance=2097152
 soundClassObject=&Fort_UI
 volumeMultiplier=0.28
 concurrencyOverrides {maxCount=2, resolutionRule=StopOldest}
}
