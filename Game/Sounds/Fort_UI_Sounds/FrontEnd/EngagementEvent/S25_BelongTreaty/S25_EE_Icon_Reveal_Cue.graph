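# Sound cue named for the icon reveal in the S25 BelongTreaty front-end engagement event, routed to the Fort_UI sound class.
# It has no root or sound nodes, so it plays no audio as it stands.
# It keeps a reduced volume multiplier and a concurrency override that allows one instance, blocking new plays and limiting rapid retriggers.
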SoundCue /Game/Sounds/Fort_UI_Sounds/FrontEnd/EngagementEvent/S25_BelongTreaty/S25_EE_Icon_Reveal_Cue
root=null
nodes []
properties {
 bOverrideConcurrency=true
 maxDistance=2097152
 soundClassObject=&Fort_UI
 volumeMultiplier=0.7
 concurrencyOverrides {maxCount=1, resolutionRule=PreventNew, retriggerTime=0.2}
}
