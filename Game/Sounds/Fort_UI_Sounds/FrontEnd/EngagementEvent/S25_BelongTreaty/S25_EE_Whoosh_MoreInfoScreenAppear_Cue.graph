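# Sound cue for the front-end UI whoosh named for the More Info screen appearing in the S25 BelongTreaty engagement event.
# A Mixer layers two medium whoosh waves from the generic UI toolkit at equal volume, played on the Fort_UI sound class at reduced overall volume.
# Its concurrency override allows one instance at a time, blocking new plays and adding a short retrigger time.
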SoundCue /Game/Sounds/Fort_UI_Sounds/FrontEnd/EngagementEvent/S25_BelongTreaty/S25_EE_Whoosh_MoreInfoScreenAppear_Cue
root=Mixer_0
nodes [
 {id=Mixer_0, children [WavePlayer_1,WavePlayer_0], properties {inputVolume [1,1]}}
 {
  id=WavePlayer_0
  properties {soundWaveAssetPtr=/Game/Sounds/Fort_UI_Sounds/FrontEnd/_GenericToolkit_/Whooshes/Misc_Whooshes/wav/UI_Toolkit_Whoosh_Med_10}
 }
 {
  id=WavePlayer_1
  properties {soundWaveAssetPtr=/Game/Sounds/Fort_UI_Sounds/FrontEnd/_GenericToolkit_/Whooshes/Misc_Whooshes/wav/UI_Toolkit_Whoosh_Med_06}
 }
]
properties {
 bOverrideConcurrency=true
 maxDistance=2097152
 soundClassObject=&Fort_UI
 volumeMultiplier=0.7
 concurrencyOverrides {maxCount=1, resolutionRule=PreventNew, retriggerTime=0.2}
}
