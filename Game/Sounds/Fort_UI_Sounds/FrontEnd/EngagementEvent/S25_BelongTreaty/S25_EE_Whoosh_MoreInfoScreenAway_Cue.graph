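# Sound cue for a front-end UI whoosh in the S25 BelongTreaty engagement event, named for the more-info screen moving away.
# A single toolkit medium whoosh wave feeds a Modulator left at default settings.
# The cue plays in the UI sound class at reduced volume.
# A concurrency override allows one instance, blocking new plays while one is active, with a short retrigger time.
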SoundCue /Game/Sounds/Fort_UI_Sounds/FrontEnd/EngagementEvent/S25_BelongTreaty/S25_EE_Whoosh_MoreInfoScreenAway_Cue
root=Modulator_1
nodes [
 {id=Modulator_1, children [WavePlayer_0]}
 {
  id=WavePlayer_0
  properties {soundWaveAssetPtr=/Game/Sounds/Fort_UI_Sounds/FrontEnd/_GenericToolkit_/Whooshes/Misc_Whooshes/wav/UI_Toolkit_Whoosh_Med_09}
 }
]
properties {
 bOverrideConcurrency=true
 maxDistance=2097152
 soundClassObject=&Fort_UI
 volumeMultiplier=0.7
 concurrencyOverrides {maxCount=1, resolutionRule=PreventNew, retriggerTime=0.2}
}
