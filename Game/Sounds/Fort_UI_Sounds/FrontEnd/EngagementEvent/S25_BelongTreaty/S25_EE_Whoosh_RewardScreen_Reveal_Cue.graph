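# UI sound cue for the reward screen reveal whoosh in the S25 BelongTreaty front-end engagement event.
# A mixer layers a toolkit medium whoosh, run through a modulator for random variation, with the party-invite appear sound reused from the ItemShop folder.
# The cue plays slightly quieter and lower-pitched on the UI sound class.
# A concurrency override allows one instance, blocking new plays and rapid retriggers so reveals do not stack.
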SoundCue /Game/Sounds/Fort_UI_Sounds/FrontEnd/EngagementEvent/S25_BelongTreaty/S25_EE_Whoosh_RewardScreen_Reveal_Cue
root=Mixer_0
nodes [
 {id=Mixer_0, children [Modulator_0,WavePlayer_3], properties {inputVolume [1,1]}}
 {id=Modulator_0, children [WavePlayer_0]}
 {
  id=WavePlayer_0
  properties {soundWaveAssetPtr=/Game/Sounds/Fort_UI_Sounds/FrontEnd/_GenericToolkit_/Whooshes/Misc_Whooshes/wav/UI_Toolkit_Whoosh_Med_10}
 }
 {id=WavePlayer_3, properties {soundWaveAssetPtr=/Game/Sounds/Fort_UI_Sounds/FrontEnd/ItemShop/wav/UI_PartyInvite_Appear}}
]
properties {
 bOverrideConcurrency=true
 maxDistance=2097152
 pitchMultiplier=0.9
 soundClassObject=&Fort_UI
 volumeMultiplier=0.8
 concurrencyOverrides {maxCount=1, resolutionRule=PreventNew, retriggerTime=0.2}
}
